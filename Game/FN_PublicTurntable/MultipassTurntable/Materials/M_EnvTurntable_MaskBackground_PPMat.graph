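# Post-process material for the environment turntable that masks the empty background with a two-color vertical gradient, applied to scene color after depth of field.
# An If node compares scene depth to a far threshold, showing the gradient at or beyond it and the original post-process input for nearer pixels.
# Offset, powered and saturated viewport Y blends the two color parameters, with zero appended as alpha.
# The multi-color lerp and Background power nodes do not feed that If.
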Material /Game/FN_PublicTurntable/MultipassTurntable/Materials/M_EnvTurntable_MaskBackground_PPMat
graph {
 properties {
  blendableLocation=BL_SceneColorAfterDOF
  blendablePriority=111
  materialDomain=MD_PostProcess
  pixelDepthOffsetMode=PDOM_Legacy
  usedShadingModels=Default Lit
 }
 nodes [
  n1 = VectorParameter() {parameterName=Color0, defaultValue=<rgba 0.06 0.0585 0.057 1>}
  n2 = VectorParameter() {parameterName=Color1, defaultValue=<rgba 0.0423 0.0497 0.0612 1>}
  n3 = ScreenPosition()
  n4 = Power(
   Base=Add(A=ComponentMask(None=@n3.ViewportUV) {g=true}, B=ScalarParameter() {defaultValue=0.4381, parameterName=add})
   Exp=ScalarParameter() {defaultValue=12, parameterName=Power}
  )
  n5 = AppendVector(A=LinearInterpolate(A=@n1.RGB, B=@n2.RGB, Alpha=Saturate(None=@n4)), B=Constant())
  If(
   A=ComponentMask(None=SceneTexture() {sceneTextureId=PPI_SceneDepth}.Color) {r=true}
   B=Constant() {r=65535}
   "A > B"=@n5
   "A == B"=@n5
   "A < B"=SceneTexture() {sceneTextureId=PPI_PostProcessInput0}.Color
  )
  MaterialFunctionCall(1=@n1.RGB, 2=Constant3Vector() {constant=<rgba 0.724 0 0.036 0>}, 3=Constant3Vector() {constant=<rgba 0.0939 0.6875 0 0>}, 4=@n2.RGB, A=@n4) {materialFunction=&Lerp_Multiple_Float3}
  Power(Base=ScalarParameter() {defaultValue=0.05, parameterName=Background}, Exp=ScalarParameter() {defaultValue=0.4545, parameterName=dsgfsd})
 ]
}
