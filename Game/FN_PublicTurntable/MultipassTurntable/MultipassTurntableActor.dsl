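; Turntable actor that captures automation screenshots of level meshes and configured spawned assets through a series of passes, then quits the game.
; BeginCapture spawns assets, sets up a camera and orders TurntablePassComponentBase components by override, running each in turn.
; Screenshot helpers frame filtered actors' bounds using rotation and FOV, ignoring the supplied location.
; TakeAutomationShot passes a hard-coded placeholder name, discarding the built level-and-index name.
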(event EventBeginPlay
  (if (Variables|Turntable|GetstartAutomatically)
    (Captures|BeginCapture (Variables|Turntable|GetinitialStartDelay))))

(event Custom|takeAutomationShot
  (params (NameOverride :type String) (passComp :type Turntable%20Pass%20Component%20Base%20Object%20Reference))
  (Automation|TakeAutomationScreenshotatCamera (Variables|Camera|GetcurrentCameraActor) "dddfdf" "sfdsfdsf" (Automation|GetDefaultScreenshotOptionsforRendering "Low" 0.0))
  (Utilities|IsValid passComp
    (:Is%20Valid
      (Camera|CaptureNextCamera passComp))))

(fn beginCapture
  (params (delay :type Float%20%28double-precision%29 :default 1.0))
  (Variables|Passes|SetcurrentPassIndex -1)
  (RenderActors|SpawnAllRenderActors)
  (Rendering|SetActorHiddenInGame (Game|GetPlayerPawn 0) true)
  (Camera|InitializeShotCamera)
  (Passes|InitializePasses)
  (Default|CallOnCaptureStarted self)
  (if (<= delay 0.0)
    (Passes|StartNextPass)
    (else
      (Utilities|Time|SetTimerbyEvent (EventDispatchers|CreateEvent :!event startNextPass) delay))))

(fn captureScreenshot
  (params (nameOverride :type String :default "Screenshot") (passComp :type Turntable%20Pass%20Component%20Base%20Object%20Reference))
  (bind _returnvalue (Game|GetCurrentLevelName true))
  (LocalVariables|Setl_levelName _returnvalue)
  (CallFunction|TakeAutomationShot (Utilities|String|Append (select (Utilities|String|NotEqualExactly(String) (Variables|GetnameOverride)) (Variables|GetnameOverride) (LocalVariables|Getl_levelName)) "_" (Utilities|String|Right (Utilities|String|Append "0000000" (Utilities|String|ToString(Integer) (Variables|Captures|GetshotIndex))) 4)) passComp)
  (Math|Integer|IncrementInt (Variables|Captures|GetshotIndex)))

(fn clearSpawnedActors
  (bind _spawnedactors (Variables|Private|GetspawnedActors))
  (for _array_element _spawnedactors
    (Actor|DestroyActor _array_element))
  (Utilities|Array|Clear _spawnedactors))

(fn endCapture
  (Default|CallOnCaptureComplete self)
  (Default|ExitGame))

(fn exitGame
  (Game|QuitGame 0))

(fn filterClass
  (params (classToCheck :type Object%20Class%20Reference) (listToCheckInside :type Array%20of%20Object%20Class%20References))
  (bind _array_element (Utilities|Array|ForEachLoopwithBreak (Variables|GetlistToCheckInside))
    (:LoopBody
      (if (Utilities|ClassIsChildOf (Variables|GetclassToCheck) _array_element)
        (return false)))
    (:Completed
      (return true))))

(fn filterMeshForUse
  (params (actorToCheck :type Actor%20Object%20Reference))
  (return (select (or (> (Utilities|Array|Length (Variables|TurntableIgnore|GetonlyCaptureActorClasses)) 0) (> (Utilities|Array|Length (Variables|TurntableIgnore|GetonlyCaptureLevelActors)) 0)) (Math|Boolean|ANDBoolean (or (Utilities|Array|ContainsItem (Variables|TurntableIgnore|GetonlyCaptureActorClasses) (Utilities|GetClass (Variables|GetactorToCheck))) (Utilities|Array|ContainsItem (Variables|TurntableIgnore|GetonlyCaptureLevelActors) (Variables|GetactorToCheck))) (not (Game|IsActorBeingDestroyed (Variables|GetactorToCheck))) (<= (Actor|GetLifeSpan (Variables|GetactorToCheck)) 0.0)) (Math|Boolean|ANDBoolean (or (Utilities|IsValid (Actor|GetComponentbyClass (Variables|GetactorToCheck) "/Script/Engine.StaticMeshComponent") :!class /Script/Engine.KismetSystemLibrary) (Utilities|IsValid (Actor|GetComponentbyClass (Variables|GetactorToCheck) "/Script/Engine.SkeletalMeshComponent") :!class /Script/Engine.KismetSystemLibrary)) (CallFunction|FilterClass (Utilities|GetClass (Variables|GetactorToCheck)) (Variables|TurntableIgnore|GetactorClassesToIgnore)) (not (Utilities|Array|ContainsItem (Variables|TurntableIgnore|GetlevelActorsToIgnore) (Variables|GetactorToCheck))) (not (Game|IsActorBeingDestroyed (Variables|GetactorToCheck))) (<= (Actor|GetLifeSpan (Variables|GetactorToCheck)) 0.0)))))

(fn getBoundsOfRenderActors
  (params (offset :type Vector) (minRadius :type Float%20%28double-precision%29))
  (Utilities|FlowControl|Sequence
    (:then_0
      (bind _actors (RenderActors|GetRenderActors))
      (for _array_element _actors
        (if (not (Variables|/Script/Engine.Actor|GetbHidden _array_element))
          (Utilities|Array|Add (LocalVariables|Getl_finalActors) _array_element))))
    (:then_1
      (bind (_center _boxextent) (Collision|GetActorArrayBounds (LocalVariables|Getl_finalActors)))
      (return (select (Variables|Turntable|GetuseThisActorsPositionAsCenter) (Transformation|GetActorLocation) (- _center (Variables|Getoffset))) (Math|Float|Max(Float) (select (> (Variables|/Game/FN_PublicTurntable/MultipassTurntable/ScreenshotPasses/TurntablePassComponentBase.TurntablePassComponentBase_C|GetboundsOverride (Utilities|Array|Get(acopy) (Variables|Passes|GetcurrentPasses) (Variables|Passes|GetcurrentPassIndex))) 0.0) (Variables|/Game/FN_PublicTurntable/MultipassTurntable/ScreenshotPasses/TurntablePassComponentBase.TurntablePassComponentBase_C|GetboundsOverride (Utilities|Array|Get(acopy) (Variables|Passes|GetcurrentPasses) (Variables|Passes|GetcurrentPassIndex))) (Math|Float|Max(Float) (select (Variables|Turntable|GetuseThisActorsPositionAsCenter) (Math|Float|float+float (Math|Vector|VectorLength _boxextent) (Math|Vector|VectorLength (- (- _center (Variables|Getoffset)) (Transformation|GetActorLocation))) (Math|Vector|VectorLength (Variables|Getoffset))) (+ (Math|Vector|VectorLength _boxextent) (Math|Vector|VectorLength (Variables|Getoffset)))) (Variables|GetminRadius))) 1e+01) _boxextent))))

(fn getFOV
  (Utilities|IsValid (Variables|Camera|GetcurrentCameraActor)
    (:Is%20Valid
      (return (Variables|/Script/Engine.CameraComponent|GetFieldOfView (Variables|/Script/Engine.CameraActor|GetCameraComponent (Variables|Camera|GetcurrentCameraActor)))))
    (:Is%20Not%20Valid
      (return 45.0))))

(fn getRenderActorBounds
  (params (offset :type Vector :default "0, 0, 0") (minRadius :type Float%20%28double-precision%29))
  (bind (_boundscenter _radius _boxextent) (RenderActors|GetBoundsOfRenderActors offset minRadius))
  (return _boundscenter _radius _boxextent))

(fn getRenderActorList
  (bind _actors (RenderActors|GetRenderActors))
  (return _actors))

(fn getRenderActors
  (Utilities|FlowControl|Sequence
    (:then_0
      (bind _outactors (Actor|GetAllActorsOfClass "/Script/Engine.Actor"))
      (for _array_element _outactors
        (bind _allowuse (RenderActors|FilterMeshforUse _array_element))
        (if _allowuse
          (Utilities|Array|Add (LocalVariables|Getl_finalActors) _array_element))))
    (:then_2
      (Utilities|Array|AppendArray (LocalVariables|Getl_finalActors) (Variables|Private|GetspawnedActors)))
    (:then_3
      (Development|PrintString "LIST OF CURRENT ACTORS" (Variables|DEBUG|GetDEBUG_PrintToScreen))
      (for _array_element_2 (LocalVariables|Getl_finalActors)
        (Development|PrintString (Utilities|GetDisplayName _array_element_2) (Variables|DEBUG|GetDEBUG_PrintToScreen) true "(R=0.000000,G=1.000000,B=0.105791,A=1.000000)"))
      (Development|PrintString ".

" (Variables|DEBUG|GetDEBUG_PrintToScreen)))
    (:then_4
      (return (LocalVariables|Getl_finalActors)))))

(fn getShotIndex
  (return 0 0 (Variables|Captures|GetshotIndex)))

(fn handlePassComplete
  (params (passComp :type Turntable%20Pass%20Component%20Base%20Object%20Reference))
  (Default|CallOnPassComplete self passComp (Variables|Passes|GetcurrentPassIndex))
  (Passes|StartNextPass)
  (return false))

(fn incrementShotIndex
  (bind _result (Math|Integer|IncrementInt (Variables|Captures|GetshotIndex))
    (:%20%20
      (return (- _result 1) _result))))

(fn initializePasses
  (if (< (Utilities|Array|Length (Variables|Passes|GetcurrentPasses)) 1)
    (for _array_element (Actor|GetComponentsByClass "/Game/FN_PublicTurntable/MultipassTurntable/ScreenshotPasses/TurntablePassComponentBase.TurntablePassComponentBase_C")
      (bind _output_get (LocalVariables|Setl_currentComp _array_element))
      (Class|TurntablePassComponentBase|Initialize _output_get)
      (if (> (Variables|/Game/FN_PublicTurntable/MultipassTurntable/ScreenshotPasses/TurntablePassComponentBase.TurntablePassComponentBase_C|GetpassOrderOverride (LocalVariables|Getl_currentComp)) -1)
        (if (< (Utilities|Array|Length (Variables|Passes|GetcurrentPasses)) 1)
          (Utilities|Array|Add (Variables|Passes|GetcurrentPasses) (LocalVariables|Getl_currentComp)) ; The first overriden should just be added to the array
          (else
            (LocalVariables|Setl_found false)
            (for (_array_element_2 _array_index) (Variables|Passes|GetcurrentPasses)
              (if (< (Variables|/Game/FN_PublicTurntable/MultipassTurntable/ScreenshotPasses/TurntablePassComponentBase.TurntablePassComponentBase_C|GetpassOrderOverride (LocalVariables|Getl_currentComp)) (Variables|/Game/FN_PublicTurntable/MultipassTurntable/ScreenshotPasses/TurntablePassComponentBase.TurntablePassComponentBase_C|GetpassOrderOverride _array_element_2))
                (LocalVariables|Setl_found true)
                (Utilities|Array|Insert (Variables|Passes|GetcurrentPasses) (LocalVariables|Getl_currentComp) _array_index) ; Found an override higher than our current, insert before this one
              ))
            (if (not (LocalVariables|Getl_found))
              (Utilities|Array|Add (Variables|Passes|GetcurrentPasses) (LocalVariables|Getl_currentComp)) ; Did not find an override less than current, add to the end
            )))
        (else
          (Utilities|Array|Add (LocalVariables|Getl_unsorted) (LocalVariables|Getl_currentComp)) ; Override not set, just add to unsorted array
        )))
    (Utilities|Array|AppendArray (Variables|Passes|GetcurrentPasses) (LocalVariables|Getl_unsorted)) ; When done sorting, concatenate the sorted and unsorted lists
  ))

(fn initializeShotCamera
  (Utilities|IsValid (Variables|Camera|GetcurrentCameraActor)
    (:Is%20Valid
      _)
    (:Is%20Not%20Valid
      (bind _returnvalue (Game|SpawnActorfromClass "/Script/Engine.CameraActor" (struct :Location "0, 0, 0" :Rotation "0, 0, 0" :Scale "1.000000,1.000000,1.000000") "AlwaysSpawn" "MultiplyWithRoot" 0 0))
      (bind _output_get (Variables|Camera|SetcurrentCameraActor _returnvalue))
      (Camera|SetConstraintAspectRatio (Variables|/Script/Engine.CameraActor|GetCameraComponent _output_get) false)))
  (Game|Player|SetViewTargetwithBlend (Game|GetPlayerController 0) (Variables|Camera|GetcurrentCameraActor) 0.0 "VTBlend_Linear" 0.0 false))

(fn initPostProcessBackground
  (Variables|/Script/Engine.PostProcessComponent|SetbEnabled (Variables|Turntable|GetusePostProcessBackground) (Variables|Default|GetPostProcess)))

(fn setCameraLocationAndRotation
  (params (NewLocation :type Vector) (NewRotation :type Rotator) (offset :type Vector) (minRadius :type Float%20%28double-precision%29))
  (Utilities|FlowControl|Sequence
    (:then_0
      (bind _outactors (Actor|GetAllActorsOfClass "/Script/Engine.StaticMeshActor"))
      (LocalVariables|Setl_finalActors _outactors)
      (bind _outactors_2 (Actor|GetAllActorsOfClass "/Script/Engine.SkeletalMeshActor"))
      (Utilities|Array|AppendArray (LocalVariables|Getl_finalActors) _outactors_2)
      (bind (_boundscenter _radius) (RenderActors|GetBoundsOfRenderActors (Variables|Getoffset) (Variables|GetminRadius)))
      (LocalVariables|Setl_boundCenter _boundscenter)
      (LocalVariables|Setl_boundRadius _radius))
    (:then_1
      (if (Utilities|Enum|Equal(Enum) (Variables|/Script/Engine.CameraComponent|GetProjectionMode (Variables|/Script/Engine.CameraActor|GetCameraComponent (Variables|Camera|GetcurrentCameraActor))) "Orthographic")
        (Camera|SetOrthoWidth (Variables|/Script/Engine.CameraActor|GetCameraComponent (Variables|Camera|GetcurrentCameraActor)) (* (LocalVariables|Getl_boundRadius) 2.0)))
      (Transformation|SetActorLocationAndRotation (Variables|Camera|GetcurrentCameraActor) (Utilities|Select :Index (Variables|/Script/Engine.CameraComponent|GetProjectionMode (Variables|/Script/Engine.CameraActor|GetCameraComponent (Variables|Camera|GetcurrentCameraActor))) :Perspective (+ (LocalVariables|Getl_boundCenter) (* (Math|Vector|GetForwardVector (Variables|GetNewRotation)) (* (/ (* (LocalVariables|Getl_boundRadius) 2.0) (Math|Trig|Tan(Degrees) (* (Variables|/Script/Engine.CameraComponent|GetFieldOfView (Variables|/Script/Engine.CameraActor|GetCameraComponent (Variables|Camera|GetcurrentCameraActor))) 0.5))) -1.0))) :Orthographic (+ (LocalVariables|Getl_boundCenter) (* (Math|Vector|GetForwardVector (Variables|GetNewRotation)) (* (/ (* (LocalVariables|Getl_boundRadius) 2.0) (Math|Trig|Tan(Degrees) (* (Variables|/Script/Engine.CameraComponent|GetFieldOfView (Variables|/Script/Engine.CameraActor|GetCameraComponent (Variables|Camera|GetcurrentCameraActor))) 0.5))) -1.0)))) (Variables|GetNewRotation) false true)
      (if (Variables|Debug|GetDEBUG_showBounds)
        (Rendering|Debug|FlushPersistentDebugLines)
        (Rendering|Debug|DrawDebugSphere (LocalVariables|Getl_boundCenter) (LocalVariables|Getl_boundRadius) (Math|Integer|Clamp(Integer) (Math|Float|Truncate (/ (LocalVariables|Getl_boundRadius) 1e+01)) 15 80) "(R=1.000000,G=0.035028,B=0.000000,A=1.000000)" 1111.0 (Math|Float|Max(Float) (/ (LocalVariables|Getl_boundRadius) 4e+02) 1.0))))))

(fn setCameraProjectionMode
  (params (isPerspective :type Boolean))
  (Camera|SetProjectionMode (Variables|/Script/Engine.CameraActor|GetCameraComponent (Variables|Camera|GetcurrentCameraActor)) (select isPerspective "Perspective" "Orthographic"))
  (return false))

(fn spawnAllRenderActors
  (RenderActors|ClearSpawnedActors)
  (for _array_element (Variables|Turntable|GetassetsToSpawnForRender)
    (bind _actorspawned (RenderActors|SpawnRenderActorfromAsset _array_element))
    (Utilities|IsValid _actorspawned
      (:Is%20Valid
        (Utilities|Array|Add (Variables|Private|GetspawnedActors) _actorspawned)))))

(fn spawnRenderActorFromAsset
  (params (assetToSpawn :type Multipass%20TT%20Spawn%20Struct%20Structure%20%28by%20ref%29))
  (bind (_assettospawn_optionalname_2_8af43b6847a8df0006c524bd766166b3 _assettospawn_asset_18_e92d0c36446b2cf7f5dea4a06224bead _assettospawn_transform_14_bde80147472a0438a675e691fde2d4de) (Utilities|Struct|BreakMultipassTTSpawnStruct (Variables|GetassetToSpawn_OptionalName_2_8AF43B6847A8DF0006C524BD766166B3)))
  (bind _spawnedactortransformoffset (Variables|Turntable|GetSpawnedActorTransformOffset))
  (bind _returnvalue (Math|Transform|MakeRelativeTransform _spawnedactortransformoffset _assettospawn_transform_14_bde80147472a0438a675e691fde2d4de))
  (bind (_assettospawn_optionalname_2_8af43b6847a8df0006c524bd766166b3_2 _assettospawn_asset_18_e92d0c36446b2cf7f5dea4a06224bead_2 _assettospawn_transform_14_bde80147472a0438a675e691fde2d4de_2 _assettospawn_animation_17_878352ff4759a636dc5481bc4f7aa50b) (Utilities|Struct|BreakMultipassTTSpawnStruct (Variables|GetassetToSpawn_OptionalName_2_8AF43B6847A8DF0006C524BD766166B3)))
  (Utilities|FlowControl|Sequence
    (:then_0
      (bind _returnvalue_2 (Game|SpawnActorfromClass "/Script/Engine.Actor" (struct :Location "0, 0, 0" :Rotation "0, 0, 0" :Scale "1.000000,1.000000,1.000000") "AlwaysSpawn" "OverrideRootScale" 0 0))
      (LocalVariables|Setl_newActor _returnvalue_2))
    (:then_1
      (bind _asstatic_mesh (Utilities|Casting|CastToStaticMesh _assettospawn_asset_18_e92d0c36446b2cf7f5dea4a06224bead))
      (bind _returnvalue_3 (AddComponent|NODE_AddStaticMeshComponent-0 :self (LocalVariables|Getl_newActor) :RelativeTransform_Location (.location _returnvalue) :RelativeTransform_Rotation (.rotation _returnvalue) :RelativeTransform_Scale (.scale _returnvalue)))
      (Components|StaticMesh|SetStaticMesh _returnvalue_3 _asstatic_mesh)
      (return (LocalVariables|Getl_newActor)))
    (:then_2
      (bind _asskeletal_mesh (Utilities|Casting|CastToSkeletalMesh _assettospawn_asset_18_e92d0c36446b2cf7f5dea4a06224bead))
      (bind _returnvalue_4 (AddComponent|NODE_AddSkeletalMeshComponent-0 :self (LocalVariables|Getl_newActor) :RelativeTransform_Location "0, 0, 0" :RelativeTransform_Rotation "0, 0, 0" :RelativeTransform_Scale (.scale _returnvalue)))
      (Components|SkinnedMesh|SetSkinnedAssetandUpdate _returnvalue_4 _asskeletal_mesh true)
      (Utilities|IsValid _assettospawn_animation_17_878352ff4759a636dc5481bc4f7aa50b
        (:Is%20Valid
          (Components|Animation|SetAnimation _returnvalue_4 _assettospawn_animation_17_878352ff4759a636dc5481bc4f7aa50b))
        (:Is%20Not%20Valid
          _))
      (return (LocalVariables|Getl_newActor)))))

(fn startNextPass
  (Utilities|FlowControl|Sequence
    (:then_0
      ;[1 1 1] Reset/Clean up before next stage
      (Camera|SetProjectionMode (Variables|/Script/Engine.CameraActor|GetCameraComponent (Variables|Camera|GetcurrentCameraActor)) "Perspective")
      (bind _asmultipass_tt_hud (Utilities|Casting|CastToMultipassTT_HUD (HUD|GetHUD (Game|GetPlayerController 0))))
      (Class|MultipassTTHUD|SetDoDrawScale _asmultipass_tt_hud false "0, 0, 0"))
    (:then_1
      (Math|Integer|IncrementInt (Variables|Passes|GetcurrentPassIndex)
        (:%20%20
          (if (Utilities|Array|IsValidIndex (Variables|Passes|GetcurrentPasses) (Variables|Passes|GetcurrentPassIndex))
            (Class|TurntablePassComponentBase|BeginPass (Utilities|Array|Get(acopy) (Variables|Passes|GetcurrentPasses) (Variables|Passes|GetcurrentPassIndex)))
            (else
              (Captures|EndCapture))))))))

(fn takeScreenshot
  (params (passComp :type Turntable%20Pass%20Component%20Base%20Object%20Reference))
  (Captures|CaptureScreenshot "Screenshot" passComp)
  (return (Variables|Captures|GetshotIndex)))

(fn takeScreenshotFromTransform
  (params (location :type Vector :default "0, 0, 0") (rotation :type Rotator :default "0, 0, 0") (offset :type Vector :default "0, 0, 0") (minRadius :type Float%20%28double-precision%29) (passComp :type Turntable%20Pass%20Component%20Base%20Object%20Reference))
  (Camera|SetCameraLocationandRotation location rotation offset minRadius)
  (Captures|CaptureScreenshot "Screenshot" passComp)
  (return (Variables|Captures|GetshotIndex)))

(fn takeScreenshotUsingCameraComponent
  (params (camera :type Camera%20Component%20Object%20Reference) (offset :type Vector :default "0, 0, 0") (minRadius :type Float%20%28double-precision%29) (passComp :type Turntable%20Pass%20Component%20Base%20Object%20Reference))
  (bind (_returnvalue_location _returnvalue_rotation) (Math|Transform|BreakTransform (Transformation|GetWorldTransform camera)))
  (Camera|SetCameraLocationandRotation _returnvalue_location _returnvalue_rotation offset minRadius)
  (Captures|CaptureScreenshot "Screenshot" passComp)
  (return (Variables|Captures|GetshotIndex)))

(fn ConstructionScript
  (CallFunction|InitPostProcessBackground))
