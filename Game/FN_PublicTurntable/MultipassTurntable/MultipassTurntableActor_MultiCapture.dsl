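; A multipass turntable actor variant that captures a list of asset classes one after another in a single game session.
; On BeginPlay it binds to its own capture-complete dispatcher, which endCapture fires to start the next capture, forming a loop.
; Each pass increments the index, hides and expires the previous actor, and spawns the next as a render actor before running the parent capture.
; It exits the game when the list is empty or exhausted.
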(event EventBeginPlay
  (Default|AssignOnCaptureComplete (EventDispatchers|CreateEvent :!event handleCaptureComplete))
  (|CallParentFunction))

(event Captures|EventendCapture
  (Default|CallOnCaptureComplete self))

(event Captures|EventbeginCapture
  (params (delay :type Float%20%28double-precision%29))
  (if (> (Utilities|Array|Length (Variables|MultipassTurntable_MultiCapture|GetMultiCaptureActorClasses)) 0)
    (MultipassTurntable_MultiCapture|SetupNextCapture)
    (|CallParentFunction 1.0)
    (else
      (Default|ExitGame))))

(fn getRenderActors
  (bind _actors (|CallParentFunction))
  (Utilities|Array|AddUnique _actors (Variables|MultipassTurntable_MultiCapture|GetcurrentCaptureActor))
  (return _actors))

(fn handleCaptureComplete
  (params (actor :type Multipass%20Turntable%20Actor%20Object%20Reference))
  (Captures|beginCapture 1.0))

(fn setupNextCapture
  (bind _currentcaptureactor (Variables|MultipassTurntable_MultiCapture|GetcurrentCaptureActor))
  (Math|Integer|IncrementInt (Variables|MultipassTurntable_MultiCapture|GetcurrentCaptureIndex)
    (:%20%20
      (Utilities|FlowControl|Sequence
        (:then_0
          (Utilities|IsValid (Variables|MultipassTurntable_MultiCapture|GetcurrentCaptureActor)
            (:Is%20Valid
              (Rendering|SetActorHiddenInGame _currentcaptureactor true)
              (Actor|SetLifeSpan _currentcaptureactor 0.01))))
        (:then_1
          (if (Utilities|Array|IsValidIndex (Variables|MultipassTurntable_MultiCapture|GetMultiCaptureActorClasses) (Variables|MultipassTurntable_MultiCapture|GetcurrentCaptureIndex))
            (MultipassTurntable_MultiCapture|SpawnNewCaptureActor)
            (else
              (Default|ExitGame))))))))

(fn spawnNewCaptureActor
  (bind _actorspawned (RenderActors|SpawnRenderActorfromAsset (Utilities|Array|Get(acopy) (Variables|MultipassTurntable_MultiCapture|GetMultiCaptureActorClasses) (Variables|MultipassTurntable_MultiCapture|GetcurrentCaptureIndex))))
  (Variables|MultipassTurntable_MultiCapture|SetcurrentCaptureActor _actorspawned))

(fn ConstructionScript
  (|CallParentFunction))
